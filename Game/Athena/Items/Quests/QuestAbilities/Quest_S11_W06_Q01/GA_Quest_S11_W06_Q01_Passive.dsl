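; Passive quest ability for a Season 11 Week 6 Fortnite quest, tracking whether the owning player is low on health.
; On activation it binds a check to the pawn's health-changed event.
; When health plus shield is 50 or below, it applies the quest gameplay effect to the player; otherwise it removes active effects granting the Athena.Quests.Elim.Lessthan50health tag.
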(event Ability|EventActivateAbility
  (|CallParentFunction)
  (Health|BindEventtoOnPawnHealthChanged (Variables|GAVars|Data|GetOwningPlayerPawn) (delegate Custom|Check%20Health)))

(event Custom|Check%20Health
  (if (<= (+ (FortPawn|GetHealth (Variables|GAVars|Data|GetOwningPlayerPawn)) (FortPawn|GetShield (Variables|GAVars|Data|GetOwningPlayerPawn))) 5e+01)
    (GameplayEffects|ApplyGameplayEffectToSelf :self (Ability|GetAbilitySystemComponent (Variables|GAVars|Data|GetOwningPlayerPawn)) :GameplayEffectClass "/Game/Athena/Items/Quests/QuestAbilities/Quest_S11_W06_Q01/GE_Quest_S11_W06_Q01.GE_Quest_S11_W06_Q01_C" :Level 0.0)
    (else
      (GameplayEffects|RemoveActiveEffectswithGrantedTags (Ability|GetAbilitySystemComponent (Variables|GAVars|Data|GetOwningPlayerPawn)) "(GameplayTags=((TagName=\"Athena.Quests.Elim.Lessthan50health\")))"))))

(event Ability|EventFailedToActivatePassiveAbility
  (params (ActorInfo :type Gameplay%20Ability%20Actor%20Info%20Structure))
  (|CallParentFunction))
